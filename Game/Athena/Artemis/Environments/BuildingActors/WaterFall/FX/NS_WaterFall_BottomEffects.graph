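# Niagara system for effects at the foot of a waterfall: surface ripples, a top flow plane, and underwater bubbles, flow textures and splashes.
# User parameters drive sizes, alignment and spawn rates, with an optional opposite-direction bubble stream and high-end ripple mesh.
# GPU bubbles and splash particles use distance-scaled spawning, while the splash mesh culls by distance.
# Ripple sprites render twice, flipped with an underwater offset, and are tinted by time-of-day curves.
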NiagaraSystem /Game/Athena/Artemis/Environments/BuildingActors/WaterFall/FX/NS_WaterFall_BottomEffects
dependencies {
 usedDataInterfaces [&NiagaraDataInterfaceColorCurve, &NiagaraDataInterfaceCurve, &NiagaraDataInterfaceCamera]
 usedDynamicInputs [
  &SetBoolByFloatComparison
  &DotProduct
  &NormalizeVector
  &Subtract_Vector
  &ReturnCameraVector
  &ConvertPositionToVector
  &Boolean_AndOperation
  &OneMinusFloat
  &ClampFloat
  &Add_Float
 ]
 usedModules [
  &SystemState
  &self:SystemSpawnScript.SystemScriptSource.SystemScriptGraph.NiagaraNodeAssignment_0.SetVariables_#1
  &self:SystemSpawnScript.SystemScriptSource.SystemScriptGraph.NiagaraNodeAssignment_1.SetVariables_#2
  &self:SystemSpawnScript.SystemScriptSource.SystemScriptGraph.NiagaraNodeAssignment_2.SetVariables_#3
 ]
}
emitters {
 Bubble {
  simTarget=GPUComputeSim
  emitterUpdateScript [
   &EmitterState()
   &SpawnRate(
    Spawn Probability=$System.SpawnRateScaleByDistance
    SpawnRate=&Multiply_Float(
     A=&Multiply_Float(A=&MakeFloatFromVector(Vector=$User.UnderWaterBubble_SpawnBoxSize, Channel=::NewEnumerator0), B=0.5556)
     B=$User.UnderWaterBubble_RateScale
    )
    Use Spawn Probability=true
   )
   &SpawnRate(
    Spawn Probability=$System.SpawnRateScaleByDistance
    SpawnGroup=1
    SpawnRate=&MakeCustomFloatFromBool(
     Bool=$User.UnderWaterBubble_DoubleDirect
     True Float=&Multiply_Float(
      A=&Multiply_Float(A=&MakeFloatFromVector(Vector=$User.UnderWaterBubble_SpawnBoxSize, Channel=::NewEnumerator0), B=0.5556)
      B=$User.UnderWaterBubble_RateScale
     )
     False Float=0
    )
    Use Spawn Probability=true
   ) {module=SpawnRate001}
  ]
  particleSpawnScript [
   &InitializeParticle(
    Color=<rgba 1 1 1 0.5>
    Lifetime Max=1.7
    Lifetime Min=0.5
    Lifetime Mode=::NewEnumerator1
    Sprite Size Mode=::NewEnumerator1
    Uniform Sprite Size Max=20
    UsePositionOffset=false
    Write Material Random=true
   )
   /SetVariables(Particles.Lifetime=&Multiply_Float(A=$Particles.Lifetime, B=$User.UnderWaterBubble_ParticleLifeScale)) {isSetParameters=true}
   &ShapeLocation(
    Box Size=$User.UnderWaterBubble_SpawnBoxSize
    MaskSpawnGroup=false
    Offset=<0 0 -200>
    Offset Mode=::NewEnumerator0
    Override Seed=false
    Shape Primitive=::NewEnumerator2
    Surface Only=false
    UseOwnerScale=true
   )
   &AddVelocity(Velocity=&MaskVectorBySpawnGroup(Masked Vector=<0 -1500 0>, Passthrough Vector=<0 1500 0>))
   &ApplyInitialForces()
   &SubUVAnimation(End Frame=3, SubUV Animation Mode:ENiagara_SubUVLookupMode=::NewEnumerator2)
  ]
  particleUpdateScript [
   &ParticleState()
   &WindForce(Wind Speed=<0 0 -1>, Wind Speed Scale=20)
   &CurlNoiseForce(Noise Strength=1500, Pan Noise=false)
   &SolveForcesAndVelocity(Clamp Velocity=false, Limit Acceleration=false)
   &ScaleSpriteSize(Uniform Curve Sprite Scale=DataInterface() {propertyValues {Curve {keys [
        {interpMode=RCIM_Cubic, tangentMode=RCTM_User, time=0, value=0}
        {interpMode=RCIM_Cubic, tangentMode=RCTM_User, time=0.5, value=1}
        {tangentMode=RCTM_User, time=1, value=0}
       ]}}})
  ]
  rendererClasses [&NiagaraSpriteRendererProperties]
  renderers [{rendererClass=&NiagaraSpriteRendererProperties}]
 }
 DirectionalBurst002 {
  simTarget=GPUComputeSim
  emitterUpdateScript [
   &EmitterState(Life Cycle Mode=::NewEnumerator1, Loop Behavior=::NewEnumerator1, Loop Duration Mode=::NewEnumerator1, UseLoopDelay=false)
   &SpawnRate(
    Spawn Probability=$System.SpawnRateScaleByDistance
    SpawnRate=&Multiply_Float(A=$User.UnderWaterSplash_SpawnBoxWidth, B=1.3333)
    Use Spawn Probability=true
   )
  ]
  particleSpawnScript [
   &InitializeParticle(
    Color=<rgba 0.5 0.5 0.5 1>
    Lifetime Max=0.2
    Lifetime Min=0.1
    Lifetime Mode=::NewEnumerator1
    Mass Max=2
    Mass Mode=::NewEnumerator3
    Sprite Size Mode=::NewEnumerator1
    Uniform Sprite Size Max=18
    Uniform Sprite Size Min=7
    UsePositionOffset=false
    Write Material Random=true
   )
   &BoxLocation(
    Box Offset=<0 0 -300>
    Box Size=&MakeVector(X=$User.UnderWaterSplash_SpawnBoxWidth, Y=600, Z=0)
    MaskSpawnGroup=false
    Module.Override Seed=false
    Rotate Box=false
    Surface Only=false
   )
   &AddVelocity(Module.Use Falloff Distance=false, Override Seed=false, Velocity Mode=::NewEnumerator1, Velocity Speed=239.7487)
  ]
  particleUpdateScript [
   &ParticleState()
   &GravityForce(Gravity=<0 0 4992.2271>)
   &Drag(Drag=&RandomRangeFloat(Minimum=0.8, Maximum=1.2), Use Linear Drag=true, Use Rotational Drag=false)
   &ScaleColor(
    Scale Alpha=&FloatFromCurve(FloatCurve=DataInterface() {propertyValues {Curve {keys [
         {arriveTangent=3.4791, interpMode=RCIM_Cubic, leaveTangent=3.4791, tangentMode=RCTM_User, time=0, value=0}
         {interpMode=RCIM_Cubic, tangentMode=RCTM_Break, time=0.3, value=1}
         {arriveTangent=-4.0201, interpMode=RCIM_Cubic, leaveTangent=-4.0201, tangentMode=RCTM_User, time=1.0011, value=0.002}
        ]}}})
    ScaleA=true
    ScaleRGB=true
   )
   &CurlNoiseForce(Noise Frequency=2, Noise Strength=6377.1436, Pan Noise=false)
   &SolveForcesAndVelocity(Clamp Velocity=false, Limit Acceleration=false)
   &ScaleSpriteSize(Uniform Curve Sprite Scale=DataInterface() {propertyValues {Curve {keys [{interpMode=RCIM_Cubic, leaveTangent=0.6708, tangentMode=RCTM_Break, time=0, value=0}, {tangentMode=RCTM_User, time=0.15, value=1}]}}})
  ]
  rendererClasses [&NiagaraSpriteRendererProperties]
  renderers [{rendererClass=&NiagaraSpriteRendererProperties}]
 }
 Empty002 {
  emitterUpdateScript [
   &EmitterState(
    Enable Distance Culling=true
    Max Distance Response=::SleepAndClearParticles
    MaxDistance=6000
    Module.bUseMaxDistance=true
    Module.bUseMinDistance=false
    Scalability Mode=::NewEnumerator1
   )
   &SpawnBurst_Instantaneous(Use Spawn Probability=false, UseLoopCountLimit=false)
  ]
  particleSpawnScript [
   &InitializeParticle(
    Color Mode=::NewEnumerator0
    Lifetime=5
    Mesh Scale=&MakeVector(X=$User.UnderWaterSplashMesh_ScaleX, Y=$User.UnderWaterSplashMesh_ScaleY, Z=1)
    Mesh Scale Mode=::NewEnumerator4
    Position Offset=<0 0 -5>
    UsePositionOffset=true
    Write Material Random=true
   )
   &InitialMeshOrientation(Mesh Axis To Orient=<0 0 -1>, Mesh Orientation Mode=::NewEnumerator2, Orientation Vector=<0 0 1>, Use Rotation Vector=false)
  ]
  particleUpdateScript [
   &ParticleState()
   &DynamicMaterialParameters(
    Index 0 Param 1=300
    Index 0 Param 2=-2
    Module.Param0WriteEnabled=true
    Module.Param1WriteEnabled=true
    Module.Param2WriteEnabled=true
    Module.Param3WriteEnabled=true
   )
  ]
  rendererClasses [&NiagaraMeshRendererProperties]
  renderers [{rendererClass=&NiagaraMeshRendererProperties, rendererIndex=0}]
 }
 FlowPlane {
  emitterUpdateScript [
   &EmitterState()
   &SpawnBurst_Instantaneous(Use Spawn Probability=false, UseLoopCountLimit=false)
  ]
  particleSpawnScript [
   &InitializeParticle(
    Color Mode=::NewEnumerator0
    Lifetime=5
    Position Offset=$User.TopFlow_Offset
    Sprite Size=&MakeVector2D(X=$User.TopFlow_Size, Y=1700)
    Sprite Size Mode=::NewEnumerator4
    UsePositionOffset=true
    Write Material Random=true
   )
   /SetVariables(Particles.SpriteAlignment=$User.TopFlow_Align, Particles.SpriteFacing=<0 0 1>) {isSetParameters=true}
  ]
  particleUpdateScript [
   &ParticleState()
   &DynamicMaterialParameters(
    Index 0 Param 1=&Divide_Float(A=1700, B=$User.TopFlow_Size)
    Module.Param0WriteEnabled=true
    Module.Param1WriteEnabled=true
    Module.Param2WriteEnabled=true
    Module.Param3WriteEnabled=true
   )
  ]
  rendererClasses [&NiagaraSpriteRendererProperties]
  renderers [{rendererClass=&NiagaraSpriteRendererProperties}]
 }
 FlowTexture {
  emitterUpdateScript [
   &EmitterState()
   &SpawnRate(
    Spawn Probability=$System.SpawnRateScaleByDistance
    SpawnRate=&Multiply_Float(
     A=&Multiply_Float(A=&MakeFloatFromVector(Vector=$User.UnderWaterBubble_SpawnBoxSize, Channel=::NewEnumerator0), B=0.0278)
     B=$User.UnderWaterBubble_RateScale
    )
    Use Spawn Probability=true
   )
   &SpawnRate(
    Spawn Probability=$System.SpawnRateScaleByDistance
    SpawnGroup=1
    SpawnRate=&MakeCustomFloatFromBool(
     Bool=$User.UnderWaterBubble_DoubleDirect
     True Float=&Multiply_Float(
      A=&Multiply_Float(A=&MakeFloatFromVector(Vector=$User.UnderWaterBubble_SpawnBoxSize, Channel=::NewEnumerator0), B=0.0278)
      B=$User.UnderWaterBubble_RateScale
     )
     False Float=0
    )
    Use Spawn Probability=true
   ) {module=SpawnRate001}
   /SetVariables(Emitter.FadeDistance=50) {isSetParameters=true}
   /SetVariables(Emitter.Fresnel=1) {isSetParameters=true}
  ]
  particleSpawnScript [
   &InitializeParticle(
    Color=<rgba 0.4 0.4 0.4 1>
    Lifetime Max=1
    Lifetime Min=0.7
    Lifetime Mode=::NewEnumerator1
    Sprite Size Max=<610.095 1025.9827>
    Sprite Size Min=<358.8442 550.1611>
    Sprite Size Mode=::NewEnumerator2
    UsePositionOffset=false
    Write Material Random=true
   )
   /SetVariables(Particles.Lifetime=&Multiply_Float(A=$Particles.Lifetime, B=$User.UnderWaterBubble_ParticleLifeScale)) {isSetParameters=true}
   &ShapeLocation(
    Box Size=$User.UnderWaterBubble_SpawnBoxSize
    MaskSpawnGroup=false
    Offset=<0 0 -270>
    Offset Mode=::NewEnumerator0
    Override Seed=false
    Shape Primitive=::NewEnumerator2
    Surface Only=false
    UseOwnerScale=true
   )
   &AddVelocity(Velocity=&MaskVectorBySpawnGroup(Masked Vector=<0 -2000 0>, Passthrough Vector=<0 2000 0>))
   &ApplyInitialForces()
  ]
  particleUpdateScript [
   &ParticleState()
   &WindForce(Wind Speed=<0 0 -1>, Wind Speed Scale=20)
   &CurlNoiseForce(Noise Strength=1500, Pan Noise=false) {enabled=false}
   &SolveForcesAndVelocity(Clamp Velocity=false, Limit Acceleration=false)
   &ScaleSpriteSize(Uniform Curve Sprite Scale=DataInterface() {propertyValues {Curve {keys [
        {interpMode=RCIM_Cubic, tangentMode=RCTM_User, time=0, value=0}
        {interpMode=RCIM_Cubic, tangentMode=RCTM_User, time=0.5, value=1}
        {tangentMode=RCTM_User, time=1, value=0}
       ]}}}) {enabled=false}
   &ScaleColor(
    Scale Alpha=&FloatFromCurve(FloatCurve=DataInterface() {propertyValues {Curve {keys [
         {arriveTangent=5.7903, interpMode=RCIM_Cubic, leaveTangent=5.7903, tangentMode=RCTM_User, time=0, value=0}
         {interpMode=RCIM_Cubic, tangentMode=RCTM_User, time=0.5, value=1}
         {arriveTangent=-4.4527, interpMode=RCIM_Cubic, tangentMode=RCTM_Break, time=1, value=0}
        ]}}})
    ScaleA=true
    ScaleRGB=true
   )
  ]
  rendererClasses [&NiagaraSpriteRendererProperties]
  renderers [{rendererClass=&NiagaraSpriteRendererProperties}]
 }
 Ripple {
  emitterUpdateScript [
   &EmitterState()
   &SpawnBurst_Instantaneous(Use Spawn Probability=false, UseLoopCountLimit=false)
  ]
  particleSpawnScript [
   &InitializeParticle(
    Color Mode=::NewEnumerator0
    Lifetime=5
    Sprite Size Mode=::NewEnumerator3
    Uniform Sprite Size=&Multiply_Float(A=700, B=$User.Ripple_Size)
    UsePositionOffset=false
    Write Material Random=true
   )
   /SetVariables(Particles.SpriteAlignment=$User.Ripple_Alignment, Particles.SpriteFacing=<0 0 1>) {isSetParameters=true}
   /SetVariables(
    Particles.SpriteAlignment2:Vector3f=&Multiply_Vector(A=$User.Ripple_Alignment, B=<-1 -1 -1>)
    Particles.SpriteFacing2:Vector3f=<0 0 -1>
   ) {isSetParameters=true}
   /SetVariables(Particles.Position2:NiagaraPosition=&Add_Vector(
     A=Linked() {linkedVariable {name=Particles.Position, type=NiagaraPosition}}
     B=&MakeVector(X=0, Y=0, Z=$User.Ripple_UnderWaterOffset)
    )) {isSetParameters=true}
  ]
  particleUpdateScript [
   &ParticleState()
   &ScaleColor(
    Scale Alpha=&FloatFromCurve(
     FloatCurve=DataInterface() {propertyValues {Curve {keys [<at 0 0.5>,<at 0.043 1>,<at 0.0841 1>,<at 0.125 0.9>,<at 0.208 0.25>,<at 0.25 0.1>,<at 0.292 0.2>,<at 0.5 0>,<at 0.87 0>,<at 0.92 1>,<at 1 0.5>]}}}
     CurveIndex=Linked() {linkedVariable {name=NPC.FortniteNPC.TimeOfDay_0-1, type=NiagaraFloat}}
    )
    Scale RGB=&MakeVectorFromLinearColorRGB(Color=&ColorFromCurve(
      ColorCurve=DataInterface() {propertyValues {
        AlphaCurve {keys [<at 0 1 RCIM_Cubic>,<at 1 0 RCIM_Cubic>]}
        BlueCurve {keys [
          <at 0 0>
          <at 0.0426 0>
          <at 0.0833 0.05>
          {arriveTangent=-0.096, interpMode=RCIM_Cubic, leaveTangent=-0.096, time=0.125, value=0}
          <at 0.1666 0.042>
          <at 0.208 0.06>
          <at 0.2495 0>
          <at 0.2922 0>
          <at 0.3363 0>
          <at 0.3809 0>
          <at 0.4182 0>
          <at 0.4589 0>
          <at 0.5 0.03>
          <at 0.58 0>
          <at 0.8333 0>
          <at 0.8698 0.01>
          <at 1 0 RCIM_Cubic>
         ]}
        GreenCurve {keys [
          <at 0 0>
          <at 0.0426 0>
          <at 0.0833 0.0479>
          {arriveTangent=-0.043, interpMode=RCIM_Cubic, leaveTangent=-0.043, time=0.125, value=0}
          <at 0.1666 0.0443>
          <at 0.208 0.0633>
          <at 0.2495 0>
          <at 0.2922 0>
          <at 0.3363 0>
          <at 0.3809 0>
          <at 0.4182 0>
          <at 0.4589 0>
          <at 0.5 0.03>
          <at 0.58 0>
          <at 0.8333 0>
          <at 0.8698 0.0083>
          <at 1 0 RCIM_Cubic>
         ]}
        RedCurve {keys [
          <at 0 0>
          <at 0.0426 0>
          <at 0.0833 0.0375>
          {arriveTangent=0.3902, interpMode=RCIM_Cubic, leaveTangent=0.3902, time=0.125, value=0}
          <at 0.1666 0.07>
          <at 0.208 0.1>
          <at 0.2495 0>
          <at 0.2922 0>
          <at 0.3363 0>
          <at 0.3809 0>
          <at 0.4182 0>
          <at 0.4589 0>
          <at 0.5 0.03>
          <at 0.58 0>
          <at 0.8333 0>
          <at 0.8698 0>
          <at 1 0 RCIM_Cubic>
         ]}
       }}
      CurveIndex=Linked() {linkedVariable {name=NPC.FortniteNPC.TimeOfDay_0-1, type=NiagaraFloat}}
     ))
    ScaleA=true
    ScaleRGB=true
   )
   &DynamicMaterialParameters(
    Index 0 Param 1=&FloatFromCurve(
     FloatCurve=DataInterface() {propertyValues {Curve {keys [<at 0 0.7>,<at 0.043 0.7>,<at 0.0841 1>,<at 0.1274 0.9>,<at 0.208 0.8>,<at 0.25 1>,<at 0.2896 0.8936>,<at 0.3288 0.7636>,<at 0.381 0.8>,<at 0.4416 0.55>,<at 0.5 0.4>,<at 0.58 0.75>,<at 0.6349 0.6>,<at 0.666 0.75>,<at 0.708 0.7>,<at 0.7483 0.8>,<at 0.799 0.85>,<at 0.9 0.75>,<at 0.96 0.7>,<at 1 0.7>]}}}
     CurveIndex=Linked() {linkedVariable {name=NPC.FortniteNPC.TimeOfDay_0-1, type=NiagaraFloat}}
    )
    Index 3 Param 3=&FloatFromCurve(
     FloatCurve=DataInterface() {propertyValues {Curve {keys [<at 0 1>,<at 0.043 1>,<at 0.0841 0.5>,<at 0.1274 0.9>,<at 0.208 0.7>,<at 0.25 1>,<at 0.2896 0.8936>,<at 0.3288 0.7636>,<at 0.381 0.8>,<at 0.4416 1>,<at 0.5 1>,<at 0.58 0.75>,<at 0.6349 1>,<at 0.666 1>,<at 0.708 1>,<at 0.7483 1>,<at 0.799 1>,<at 0.9 1>,<at 0.96 1>,<at 1 1>]}}}
     CurveIndex=Linked() {linkedVariable {name=NPC.FortniteNPC.TimeOfDay_0-1, type=NiagaraFloat}}
    )
    Index 3 Param 4=&FloatFromCurve(
     FloatCurve=DataInterface() {propertyValues {Curve {keys [<at 0 0>,<at 0.38 0>,<at 0.41 0.5>,<at 0.625 0.8>,<at 0.75 1>,<at 0.8 0>,<at 0.9 0.3>,<at 1 0>]}}}
     CurveIndex=Linked() {linkedVariable {name=NPC.FortniteNPC.TimeOfDay_0-1, type=NiagaraFloat}}
    )
    Module.Param0WriteEnabled=true
    Module.Param0WriteEnabled3=true
    Module.Param1WriteEnabled=true
    Module.Param1WriteEnabled3=true
    Module.Param2WriteEnabled=true
    Module.Param2WriteEnabled3=true
    Module.Param3WriteEnabled=true
    Module.Param3WriteEnabled3=true
    Write Parameter Index 3=true
   )
  ]
  rendererClasses [&NiagaraSpriteRendererProperties]
  renderers [
   {rendererClass=&NiagaraSpriteRendererProperties}
   {rendererClass=&NiagaraSpriteRendererProperties, rendererIndex=1}
  ]
 }
 Ripple001 {
  emitterUpdateScript [
   &EmitterState()
   &SpawnBurst_Instantaneous(
    Spawn Count=&MakeCustomIntFromBool(Bool=$User.RippleHighEnd_Enable, True INT=1, False INT=0)
    Use Spawn Probability=false
    UseLoopCountLimit=false
   )
  ]
  particleSpawnScript [&InitializeParticle(
    Color Mode=::NewEnumerator0
    Lifetime=5
    Mesh Scale Mode=::NewEnumerator3
    Mesh Uniform Scale=7.5
    UsePositionOffset=true
    Write Material Random=true
   )]
  particleUpdateScript [
   &ParticleState()
   &ScaleMeshSize(Scale Factor=&VectorFromFloat(Value=$User.RippleHighEnd_Size))
  ]
  rendererClasses [&NiagaraMeshRendererProperties]
  renderers [{rendererClass=&NiagaraMeshRendererProperties, rendererIndex=0}]
 }
}
summary {
 systemName=NS_WaterFall_BottomEffects
 emitters [
  {emitterName=Ripple, rendererClasses [&NiagaraSpriteRendererProperties]}
  {emitterName=Ripple001, rendererClasses [&NiagaraMeshRendererProperties]}
  {emitterName=Bubble, simTarget=GPUComputeSim, rendererClasses [&NiagaraSpriteRendererProperties]}
  {emitterName=FlowTexture, rendererClasses [&NiagaraSpriteRendererProperties]}
  {emitterName=DirectionalBurst002, simTarget=GPUComputeSim, rendererClasses [&NiagaraSpriteRendererProperties]}
  {emitterName=Empty002, rendererClasses [&NiagaraMeshRendererProperties]}
  {emitterName=FlowPlane, rendererClasses [&NiagaraSpriteRendererProperties]}
 ]
 userVariables [
  User.Foam_Enable {type=NiagaraBool}
  User.Foam_SpawnRate {type=NiagaraFloat, defaultValue=0, description=""}
  User.Ripple_Alignment {type=Vector3f, description="", defaultValue {struct=Vector3f, value=<0 1 0>}}
  User.Ripple_Material {
   type=MaterialInterface
   description=""
   defaultValue {struct=NiagaraExt_VariableValue_Object, value {object=null, objectClass=/Script/Engine.MaterialInterface}}
  }
  User.Ripple_Size {type=NiagaraFloat, defaultValue=0, description=""}
  User.Ripple_UnderWaterOffset {type=NiagaraFloat, defaultValue=0, description=""}
  User.RippleHighEnd_Enable {type=NiagaraBool}
  User.RippleHighEnd_Material {
   type=MaterialInterface
   description=""
   defaultValue {struct=NiagaraExt_VariableValue_Object, value {object=null, objectClass=&MaterialInterface}}
  }
  User.RippleHighEnd_Size {type=NiagaraFloat, defaultValue=0, description=""}
  User.TopFlow_Align {type=Vector3f, description="", defaultValue {struct=Vector3f, value=<0 -1 0>}}
  User.TopFlow_Offset {type=Vector3f, description="", defaultValue {struct=Vector3f, value=<0 0 0>}}
  User.TopFlow_Size {type=NiagaraFloat, defaultValue=0, description=""}
  User.UnderWaterBubble_DoubleDirect {type=NiagaraBool}
  User.UnderWaterBubble_ParticleLifeScale {type=NiagaraFloat, defaultValue=0, description=""}
  User.UnderWaterBubble_RateScale {type=NiagaraFloat, defaultValue=0, description=""}
  User.UnderWaterBubble_SpawnBoxSize {type=Vector3f, description="", defaultValue {struct=Vector3f, value=<0 0 0>}}
  User.UnderWaterBubbleAndSplashP_Enable {type=NiagaraBool}
  User.UnderWaterSplash_SpawnBoxWidth {type=NiagaraFloat, defaultValue=0, description=""}
  User.UnderWaterSplashMesh_Mat {
   type=MaterialInterface
   description=""
   defaultValue {struct=NiagaraExt_VariableValue_Object, value {object=null, objectClass=&MaterialInterface}}
  }
  User.UnderWaterSplashMesh_ScaleX {type=NiagaraFloat, defaultValue=0, description=""}
  User.UnderWaterSplashMesh_ScaleY {type=NiagaraFloat, defaultValue=0, description=""}
 ]
}
