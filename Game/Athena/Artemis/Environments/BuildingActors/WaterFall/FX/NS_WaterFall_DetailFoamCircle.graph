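# Niagara system for detail foam at a waterfall, built from a single sprite emitter named Foam.
# Translucent-tinted, randomly sized, short-lived particles spawn continuously in a thin box and get an initial velocity that drag slows.
# Sprite facing is set straight up so the foam can lie flat.
# A dynamic material parameter is driven by a zero-to-one curve, leaving the foam's animation to the material.
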NiagaraSystem /Game/Athena/Artemis/Environments/BuildingActors/WaterFall/FX/NS_WaterFall_DetailFoamCircle
dependencies {usedDataInterfaces [&NiagaraDataInterfaceCurve], usedDynamicInputs [], usedModules [&SystemState]}
emitters {Foam {
  emitterUpdateScript [
   &EmitterState()
   &SpawnRate(SpawnRate=10, Use Spawn Probability=false)
  ]
  particleSpawnScript [
   &InitializeParticle(
    Color=<rgba 0.5068 0.6485 0.7 0.5>
    Lifetime Max=1.5
    Lifetime Mode=::NewEnumerator1
    Sprite Size Mode=::NewEnumerator1
    Uniform Sprite Size Max=497.287
    Uniform Sprite Size Min=288.2764
    UsePositionOffset=false
    Write Material Random=true
   )
   &BoxLocation(Box Size=<273.0699 255.9636 1>, MaskSpawnGroup=false, Module.Override Seed=false, Rotate Box=false, Surface Only=false)
   &AddVelocity(Module.Use Falloff Distance=false, Override Seed=false, Velocity Mode=::NewEnumerator1, Velocity Speed=376.7971)
   &ApplyInitialForces()
  ]
  particleUpdateScript [
   &ParticleState()
   /SetVariables(Particles.SpriteFacing=<0 0 1>) {isSetParameters=true}
   &DynamicMaterialParameters(
    Index 0 Param 2=1
    Index 0 Param 3=1
    Index 0 Param 4=&FloatFromCurve(FloatCurve=DataInterface() {propertyValues {Curve {keys [{tangentMode=RCTM_User, time=0, value=0}, {tangentMode=RCTM_User, time=1, value=1}]}}})
    Module.Param0WriteEnabled=true
    Module.Param1WriteEnabled=true
    Module.Param2WriteEnabled=true
    Module.Param3WriteEnabled=true
   )
   &Drag(Use Linear Drag=true, Use Rotational Drag=false)
   &SolveForcesAndVelocity(Clamp Velocity=false, Limit Acceleration=false)
  ]
  rendererClasses [&NiagaraSpriteRendererProperties]
  renderers [{rendererClass=&NiagaraSpriteRendererProperties}]
 }}
summary {
 systemName=NS_WaterFall_DetailFoamCircle
 emitters [{emitterName=Foam, rendererClasses [&NiagaraSpriteRendererProperties]}]
 userVariables []
}
