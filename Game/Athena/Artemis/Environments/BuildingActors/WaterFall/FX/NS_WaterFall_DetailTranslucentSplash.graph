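# Niagara system for a translucent splash detail on the Artemis waterfall, built from one continuously spawning sprite emitter.
# Short-lived, faint blue-tinted sprites spawn inside a small sphere with randomized size, rotation and alpha, launch with added upward velocity, then fall under strong gravity.
# A curve drives a dynamic material parameter downward over each particle's life.
# Sprites spin at a random rate, and User.Size scales their size.
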NiagaraSystem /Game/Athena/Artemis/Environments/BuildingActors/WaterFall/FX/NS_WaterFall_DetailTranslucentSplash
dependencies {usedDataInterfaces [&NiagaraDataInterfaceCurve], usedDynamicInputs [], usedModules [&SystemState]}
emitters {Empty {
  emitterUpdateScript [
   &EmitterState()
   &SpawnRate(SpawnRate=21.6831, Use Spawn Probability=false)
  ]
  particleSpawnScript [
   &InitializeParticle(
    Color Maximum=<rgba 0.3791 0.5075 0.6 0.075>
    Color Minimum=<rgba 0.3791 0.5075 0.6 0.035>
    Color Mode=::NewEnumerator2
    Lifetime Max=0.5
    Lifetime Min=0.35
    Lifetime Mode=::NewEnumerator1
    Sprite Rotation Mode=::NewEnumerator2
    Sprite Size Mode=::NewEnumerator1
    Uniform Sprite Size Max=267.5032
    Uniform Sprite Size Min=156.0466
    UsePositionOffset=false
    Write Material Random=true
   )
   &SphereLocation(
    MaskSpawnGroup=false
    Override Seed=false
    Sphere Radius=38.1194
    Surface Only=false
    UseNonUniformScale=false
    UseOffset=false
    UseSphereOrientation=false
   )
   &AddVelocity(Module.Use Falloff Distance=false, Override Seed=false, Velocity Mode=::NewEnumerator1, Velocity Speed=361.7095)
   &AddVelocity(Velocity=<0 0 531.6852>) {module=AddVelocity001}
   &ApplyInitialForces()
  ]
  particleUpdateScript [
   &ParticleState()
   &DynamicMaterialParameters(
    Index 0 Param 1=&FloatFromCurve(FloatCurve=DataInterface() {propertyValues {Curve {keys [
         {interpMode=RCIM_Cubic, leaveTangent=-2.2915, tangentMode=RCTM_Break, time=0.2, value=1}
         {arriveTangent=-1.905, interpMode=RCIM_Cubic, tangentMode=RCTM_Break, time=1, value=-1}
        ]}}})
    Module.Param0WriteEnabled=true
    Module.Param1WriteEnabled=true
    Module.Param2WriteEnabled=true
    Module.Param3WriteEnabled=true
   )
   &GravityForce(Gravity=<0 0 -2826.0559>)
   &SolveForcesAndVelocity(Clamp Velocity=false, Limit Acceleration=false)
   &SpriteRotationRate(Rotation Rate=&RandomRangeFloat(Minimum=-15, Maximum=15))
   &ScaleSpriteSize(Scale Sprite Size Mode=::NewEnumerator0, Uniform Scale Factor=$User.Size)
  ]
  rendererClasses [&NiagaraSpriteRendererProperties]
  renderers [{rendererClass=&NiagaraSpriteRendererProperties}]
 }}
summary {
 systemName=NS_WaterFall_DetailTranslucentSplash
 emitters [{emitterName=Empty, rendererClasses [&NiagaraSpriteRendererProperties]}]
 userVariables [User.Size {type=NiagaraFloat, defaultValue=1, description=""}]
}
