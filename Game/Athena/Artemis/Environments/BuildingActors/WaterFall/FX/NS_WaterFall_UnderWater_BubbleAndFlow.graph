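# Niagara effect for underwater waterfall bubbles and flow, built from GPU bubble sprites and larger, alpha-faded flow sprites.
# Both emitters spawn in a user-sized box offset down by half its height, launch along +Y with wind force, with curl noise only on bubbles.
# Spawn rates scale with box X width and RateScale and fall to zero with LOD distance.
# RateScale and box size default to zero, so nothing spawns without overrides.
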NiagaraSystem /Game/Athena/Artemis/Environments/BuildingActors/WaterFall/FX/NS_WaterFall_UnderWater_BubbleAndFlow
dependencies {
 usedDataInterfaces [&NiagaraDataInterfaceCurve, &NiagaraDataInterfaceCamera]
 usedDynamicInputs [
  &SetBoolByFloatComparison
  &DotProduct
  &NormalizeVector
  &Subtract_Vector
  &ReturnCameraVector
  &ConvertPositionToVector
 ]
 usedModules [
  &SystemState
  &self:SystemSpawnScript.SystemScriptSource.SystemScriptGraph.NiagaraNodeAssignment_0.SetVariables_#1
 ]
}
emitters {
 Bubble {
  simTarget=GPUComputeSim
  emitterUpdateScript [
   &EmitterState()
   &SpawnRate(
    SpawnRate=&Multiply_Float(
     A=&Multiply_Float(
      A=&Multiply_Float(A=&MakeFloatFromVector(Vector=$User.SpawnBoxSize, Channel=::NewEnumerator0), B=0.5556)
      B=$User.RateScale
     )
     B=&OneMinusFloat(Float=&ClampFloat(Float=&Divide_Float(A=&Add_Float(A=$Engine.Owner.LODDistance, B=-500), B=2000)))
    )
    Use Spawn Probability=false
   )
  ]
  particleSpawnScript [
   &InitializeParticle(
    Color=<rgba 1 1 1 0.5>
    Lifetime Max=1.7
    Lifetime Min=0.5
    Lifetime Mode=::NewEnumerator1
    Sprite Size Mode=::NewEnumerator1
    Uniform Sprite Size Max=20
    UsePositionOffset=false
    Write Material Random=true
   )
   /SetVariables(Particles.Lifetime=&Multiply_Float(A=$Particles.Lifetime, B=$User.ParticleLifeScale)) {isSetParameters=true}
   &ShapeLocation(
    Box Size=$User.SpawnBoxSize
    MaskSpawnGroup=false
    Offset=&MakeVector(
     X=0
     Y=0
     Z=&Divide_Float(A=&MakeFloatFromVector(Vector=$User.SpawnBoxSize, Channel=::NewEnumerator2), B=-2)
    )
    Offset Mode=::NewEnumerator0
    Override Seed=false
    Shape Primitive=::NewEnumerator2
    Surface Only=false
    UseOwnerScale=true
   )
   &AddVelocity(Velocity=<0 1500 0>)
   &ApplyInitialForces()
   &SubUVAnimation(End Frame=3, SubUV Animation Mode:ENiagara_SubUVLookupMode=::NewEnumerator2)
  ]
  particleUpdateScript [
   &ParticleState()
   &WindForce(Wind Speed=<0 0 -1>, Wind Speed Scale=20)
   &CurlNoiseForce(Noise Strength=1500, Pan Noise=false)
   &SolveForcesAndVelocity(Clamp Velocity=false, Limit Acceleration=false)
   &ScaleSpriteSize(Uniform Curve Sprite Scale=DataInterface() {propertyValues {Curve {keys [
        {interpMode=RCIM_Cubic, tangentMode=RCTM_User, time=0, value=0}
        {interpMode=RCIM_Cubic, tangentMode=RCTM_User, time=0.5, value=1}
        {tangentMode=RCTM_User, time=1, value=0}
       ]}}})
  ]
  rendererClasses [&NiagaraSpriteRendererProperties]
  renderers [{rendererClass=&NiagaraSpriteRendererProperties}]
 }
 FlowTexture {
  emitterUpdateScript [
   &EmitterState()
   &SpawnRate(
    SpawnRate=&Multiply_Float(
     A=&Multiply_Float(
      A=&Multiply_Float(A=&MakeFloatFromVector(Vector=$User.SpawnBoxSize, Channel=::NewEnumerator0), B=0.0278)
      B=$User.RateScale
     )
     B=&OneMinusFloat(Float=&ClampFloat(Float=&Divide_Float(A=&Add_Float(A=$Engine.Owner.LODDistance, B=-500), B=2000)))
    )
    Use Spawn Probability=false
   )
   /SetVariables(Emitter.FadeDistance=50) {isSetParameters=true}
   /SetVariables(Emitter.Fresnel=1) {isSetParameters=true}
  ]
  particleSpawnScript [
   &InitializeParticle(
    Color=<rgba 0.4 0.4 0.4 1>
    Lifetime Max=1
    Lifetime Min=0.7
    Lifetime Mode=::NewEnumerator1
    Sprite Size Max=<610.095 1025.9827>
    Sprite Size Min=<358.8442 550.1611>
    Sprite Size Mode=::NewEnumerator2
    UsePositionOffset=false
    Write Material Random=true
   )
   /SetVariables(Particles.Lifetime=&Multiply_Float(A=$Particles.Lifetime, B=$User.ParticleLifeScale)) {isSetParameters=true}
   &ShapeLocation(
    Box Size=$User.SpawnBoxSize
    MaskSpawnGroup=false
    Offset=&MakeVector(
     X=0
     Y=0
     Z=&Divide_Float(A=&MakeFloatFromVector(Vector=$User.SpawnBoxSize, Channel=::NewEnumerator2), B=-2)
    )
    Offset Mode=::NewEnumerator0
    Override Seed=false
    Shape Primitive=::NewEnumerator2
    Surface Only=false
    UseOwnerScale=true
   )
   &AddVelocity(Velocity=<0 2099.032 0>)
   &ApplyInitialForces()
  ]
  particleUpdateScript [
   &ParticleState()
   &WindForce(Wind Speed=<0 0 -1>, Wind Speed Scale=20)
   &CurlNoiseForce(Noise Strength=1500, Pan Noise=false) {enabled=false}
   &SolveForcesAndVelocity(Clamp Velocity=false, Limit Acceleration=false)
   &ScaleSpriteSize(Uniform Curve Sprite Scale=DataInterface() {propertyValues {Curve {keys [
        {interpMode=RCIM_Cubic, tangentMode=RCTM_User, time=0, value=0}
        {interpMode=RCIM_Cubic, tangentMode=RCTM_User, time=0.5, value=1}
        {tangentMode=RCTM_User, time=1, value=0}
       ]}}}) {enabled=false}
   &ScaleColor(
    Scale Alpha=&FloatFromCurve(FloatCurve=DataInterface() {propertyValues {Curve {keys [
         {arriveTangent=5.7903, interpMode=RCIM_Cubic, leaveTangent=5.7903, tangentMode=RCTM_User, time=0, value=0}
         {interpMode=RCIM_Cubic, tangentMode=RCTM_User, time=0.5, value=1}
         {arriveTangent=-4.4527, interpMode=RCIM_Cubic, tangentMode=RCTM_Break, time=1, value=0}
        ]}}})
    ScaleA=true
    ScaleRGB=true
   )
  ]
  rendererClasses [&NiagaraSpriteRendererProperties]
  renderers [{rendererClass=&NiagaraSpriteRendererProperties}]
 }
}
summary {
 systemName=NS_WaterFall_UnderWater_BubbleAndFlow
 emitters [
  {emitterName=Bubble, simTarget=GPUComputeSim, rendererClasses [&NiagaraSpriteRendererProperties]}
  {emitterName=FlowTexture, rendererClasses [&NiagaraSpriteRendererProperties]}
 ]
 userVariables [
  User.ParticleLifeScale {type=NiagaraFloat, defaultValue=1, description=""}
  User.RateScale {type=NiagaraFloat, defaultValue=0, description=""}
  User.SpawnBoxSize {type=Vector3f, description="", defaultValue {struct=Vector3f, value=<0 0 0>}}
 ]
}
