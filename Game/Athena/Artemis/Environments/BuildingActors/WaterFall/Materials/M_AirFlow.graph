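# Translucent unlit Niagara sprite material for an air flow effect in the waterfall environment set, with index-of-refraction distortion.
# A panning Perlin normal map, time-offset by a dynamic parameter, warps the center-scaled UVs of a spark texture.
# That texture, particle alpha, noise and a square edge mask form a depth-faded mask.
# The mask drives a refraction-style value above one, blended by camera distance, while particle color feeds an emissive adjustment.
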Material /Game/Athena/Artemis/Environments/BuildingActors/WaterFall/Materials/M_AirFlow
graph {
 properties {
  bUsedWithNiagaraSprites=true
  blendMode=BLEND_Translucent
  pixelDepthOffsetMode=PDOM_Legacy
  refractionMethod=RM_IndexOfRefraction
  shadingModel=MSM_Unlit
  translucencyLightingMode=TLM_VolumetricDirectional
  usedShadingModels=Unlit
 }
 nodes [
  Constant3Vector()
  n1 = MaterialFunctionCall(Float2=TextureCoordinate()) {materialFunction=&BreakOutFloat2Components}
  n2 = Subtract(A=@n1.R) {constB=0.5}
  n3 = Subtract(A=OneMinus(None=@n1.R)) {constB=0.5}
  n4 = Max(A=Max(A=Max(A=@n2, B=@n3), B=Subtract(A=@n1.G) {constB=0.5}), B=Subtract(A=OneMinus(None=@n1.G)) {constB=0.5})
  n5 = OneMinus(None=Power(Base=Multiply(A=@n4) {constB=2}) {constExponent=14})
  n6 = ParticleColor()
  n7 = TextureCoordinate()
  n8 = VectorParameter() {parameterName=Param9, defaultValue=<rgba 0.25 0.5 0 0.2>}
  n9 = AppendVector(A=@n8.R, B=@n8.R)
  n10 = AppendVector(A=@n8.B, B=@n8.B)
  n11 = DynamicParameter()
  n12 = Add(A=@n11.R, B=Time() {bOverride_Period=true, period=333})
  n13 = TextureSample(UVs=Add(A=Multiply(A=@n7, B=@n9), B=Multiply(A=@n10, B=@n12))) {samplerType=SAMPLERTYPE_Normal, texture=/Game/Characters/Player/Male/Large/Bodies/M_LRG_Clash_V/Textures/Simple_Perlin_N}
  n14 = AppendVector(A=@n13.R, B=@n13.R)
  n15 = VectorParameter() {parameterName=Param91, defaultValue=<rgba 0.25 0.1 0 0>}
  n16 = AppendVector(A=@n15.R, B=@n15.R)
  n17 = MaterialFunctionCall(UVs=Add(A=@n7, B=Multiply(A=@n14, B=@n16)), Texture Scale=ScalarParameter() {defaultValue=0.8, parameterName=s}) {materialFunction=&ScaleUVsByCenter}
  n18 = MaterialFunctionCall!(TexCoord=@n17.UVs, ClampChannel=Constant() {r=1}) {materialFunction=/Game/Effects/Fort_Effects/MaterialFunctions/MF_FX_TilingMethod}
  n19 = TextureSample(UVs=@n17.UVs) {texture=/Game/CharClasses/Soldier/Traits/AdvancedTactics/Abilities/GoingCommando/Textures/T_SparkDirection}
  n20 = Power(Base=@n19.RGB, Exp=ScalarParameter() {defaultValue=0.5, parameterName=R}) {constExponent=0.5}
  n21 = Multiply(A=Multiply(A=@n6.A, B=Multiply(A=@n18.Result, B=@n20)), B=Add(A=Saturate(None=@n13.R), B=@n11.G))
  n22 = Power(
   Base=DepthFade(Opacity=Multiply(A=@n5, B=@n21), FadeDistance=ScalarParameter() {defaultValue=50, parameterName=DF})
   Exp=ScalarParameter() {parameterName=RefP}
  )
  n23 = ScalarParameter() {parameterName=RefS}
  n24 = MaterialFunctionCall(Vector 3=Subtract(A=WorldPosition().XYZ, B=CameraPositionWS())) {materialFunction=/Engine/Functions/Engine_MaterialFunctions02/Utility/VectorLength}
  n25 = VectorParameter() {defaultValue=<rgba 2048 0 0 1>}
  LinearInterpolate(A=Add(B=Multiply(A=@n22, B=@n23)) {constA=1}, Alpha=Saturate(None=Divide(A=@n24.V3 Length, B=@n25.R)))
  MaterialFunctionCall(Emissive=@n6.RGB) {materialFunction=&MF_FX_EmissiveAdjustments}
  ParticlePositionWS()
  Power(Base=@n19.RGB)
  TextureCoordinate()
  TextureCoordinate()
 ]
}
