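# Masked, lit material for waterfall bottom-splash meshes, built for use on Niagara mesh particles.
# Scrolling noise normals drive masked world position offset, and blend with curve-atlas and foam normals for surface detail.
# Curve-atlas rows also supply color and view-dependent fake lighting, while a PortalUV switch selects radial UVs and masks.
# Opacity uses temporal dither with near-camera fade, and mobile shading paths get separate offset and dynamic-parameter emissive.
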Material /Game/Athena/Artemis/Environments/BuildingActors/WaterFall/Materials/M_BottomSplashMesh
graph {
 properties {
  bUsedWithNiagaraMeshParticles=true
  blendMode=BLEND_Masked
  pixelDepthOffsetMode=PDOM_Legacy
  translucencyLightingMode=TLM_Surface
  usedShadingModels=Default Lit
 }
 nodes [
  Add(
   A=ScalarParameter() {defaultValue=1, group=000.USE ABOVE PARAMS TO TWEAK COLOR, parameterName=_}
   B=ScalarParameter() {group=001.FOLLOWING PARAMS ARE NOT INTUITIVE, parameterName=__}
  )
  Constant() {r=0.05}
  n1 = CurveAtlasRowParameter(CurveTime=NamedRerouteUsage()) {
   atlas=/Game/Athena/Artemis/Environments/BuildingActors/WaterFall/Materials/Texture/CurveAtlas/CLA_WaterFall
   curve=/Game/Athena/Artemis/Environments/BuildingActors/WaterFall/Materials/Texture/CurveAtlas/CL_WPOCurve4
   defaultValue=7
   group=05.Normal
   parameterName=NormalCurve
  }
  n2 = Normalize(VectorInput=Multiply(A=Add(A=@n1) {constB=-0.5}) {constB=2})
  n3 = Constant3Vector() {constant=<rgba 1 -1 1 0>}
  n4 = MaterialFunctionCall(Float2=NamedRerouteUsage()) {materialFunction=&BreakOutFloat2Components}
  n5 = MaterialFunctionCall(
   Normal=If(
    A=Multiply(A=@n4.G) {constB=2}
    B=Constant() {r=1}
    "A > B"=@n2
    "A == B"=Constant3Vector() {constant=<rgba 0 0 1 0>}
    "A < B"=Multiply(A=@n2, B=@n3)
   )
   Flatness=ScalarParameter() {defaultValue=0.2, group=05.Normal, parameterName=CurveNormal_Flat}
  ) {materialFunction=&FlattenNormal}
  n6 = NamedRerouteUsage()
  n7 = LinearInterpolate(B=ScalarParameter() {defaultValue=0.5, group=05.Normal, parameterName=FromNoiseNormalB_NormalFlat}, Alpha=NamedRerouteUsage()) {constA=1}
  n8 = VectorParameter() {group=05.Normal, parameterName=FoamTexture_ScaleAndSpeed, defaultValue=<rgba 2 0.7 0 -0.6>}
  n9 = Multiply(A=NamedRerouteUsage(), B=AppendVector(A=@n8.R, B=@n8.R))
  n10 = Multiply(A=AppendVector(A=@n8.B, B=@n8.B), B=NamedRerouteUsage())
  n11 = TextureSample(UVs=Add(A=NamedRerouteUsage(), B=Add(A=@n9, B=@n10))) {
   samplerType=SAMPLERTYPE_LinearColor
   texture=/Game/Athena/Artemis/Environments/BuildingActors/WaterFall/Materials/Texture/T_MergedFoamTexture
  }
  n12 = Multiply(A=Subtract(A=AppendVector(A=@n11.R, B=@n11.R)) {constB=0.5}) {constB=2}
  MaterialFunctionCall(
   BaseNormal=MaterialFunctionCall(
    BaseNormal=@n5.Result
    AdditionalNormal=MaterialFunctionCall(Normal=@n6, Flatness=@n7) {materialFunction=&FlattenNormal}.Result
   ) {materialFunction=&BlendAngleCorrectedNormals}.Result
   AdditionalNormal=MaterialFunctionCall(Normal=DeriveNormalZ(InXY=@n12), Flatness=ScalarParameter() {defaultValue=0.93, group=05.Normal, parameterName=FoamNormal_Flat}) {materialFunction=&FlattenNormal}.Result
  ) {materialFunction=&BlendAngleCorrectedNormals}
  n13 = Constant()
  n14 = DynamicParameter() {parameterIndex=3, defaultValue=<rgba 0.6561 0.8854 1 0.1>, paramNames [ES3EmissiveC, ES3EmissiveC, ES3EmissiveC, ES3EmissiveScale]}
  n15 = Multiply(
   A=@n14.ES3EmissiveScale
   B=MaterialFunctionCall(X=@n14.ES3EmissiveC, Y=@n14.ES3EmissiveC, Z=@n14.ES3EmissiveC) {materialFunction=&MakeFloat3}.Result
  )
  MaterialFunctionCall(Emissive=Add(A=ShadingPathSwitch(Default=@n13, Deferred=@n13, Forward=@n13, Mobile=@n15), B=NamedRerouteUsage())) {materialFunction=&MF_FX_EmissiveAdjustments}
  n16 = CurveAtlasRowParameter(CurveTime=Constant() {r=0.01}) {
   atlas=&CLA_WaterFall
   curve=/Game/Athena/Artemis/Environments/BuildingActors/WaterFall/Prefab/S21/CL_SplashMeshColor
   defaultValue=19
   parameterName=Curve_Color
  }
  n17 = LinearInterpolate(A=Constant(), B=@n16, Alpha=NamedRerouteUsage())
  Multiply(A=@n17, B=@n15)
  NamedRerouteDeclaration(None=Multiply(A=Time() {bOverride_Period=true, period=300}, B=ScalarParameter() {defaultValue=0.9, group=00.Time, parameterName=TimeScale})) {name=Time, nodeColor=<rgba 0.2891 1 0 1>}
  n18 = Normalize(VectorInput=Subtract(A=ObjectPositionWS(), B=CameraPositionWS()))
  n19 = Constant3Vector() {constant=<rgba 0 -1 0 0>}
  n20 = DotProduct(A=@n18, B=Normalize(VectorInput=Transform(None=@n19) {transformSourceType=TRANSFORMSOURCE_Local}))
  n21 = StaticSwitchParameter(True=Constant(), False=Saturate(None=Subtract(A=@n20) {constB=-0.4})) {parameterName=PortalUV}
  n22 = Normalize(VectorInput=Transform(None=Constant3Vector() {constant=<rgba 0 0 -1 0>}) {transformSourceType=TRANSFORMSOURCE_Local})
  n23 = Saturate(None=Subtract(A=DotProduct(A=@n18, B=@n22)) {constB=0})
  n24 = CurveAtlasRowParameter(CurveTime=Constant() {r=0.99}) {
   atlas=&CLA_WaterFall
   curve=&CL_SplashMeshColor
   defaultValue=19
   parameterName=Curve_Color
  }
  NamedRerouteDeclaration(None=Multiply(A=Max(A=@n21, B=@n23), B=@n24)) {name=FakeSHLighting, nodeColor=<rgba 0.6875 0 1 1>}
  n25 = VectorParameter() {parameterName=DefaultNormalWhenDisableCurveNormal, defaultValue=<rgba 0 0 1 1>}
  NamedRerouteDeclaration(None=LinearInterpolate(
    A=@n5.Result
    B=Transform(None=@n25.RGB) {transformSourceType=TRANSFORMSOURCE_Local, transformType=TRANSFORM_Tangent}
    Alpha=ScalarParameter() {group=05.Normal, parameterName=DisableCurveNormal}
   )) {name=Final_Normal, nodeColor=<rgba 1 0 0.1172 1>}
  n26 = NamedRerouteUsage()
  n27 = VectorParameter() {group=03.WPO, parameterName=FromNoiseNormalB_OffsetAndPower, defaultValue=<rgba 0 0 -0.3 3>}
  n28 = Multiply(
   A=VectorParameter() {group=03.WPO, parameterName=FromNoiseNormalB_Multiply, defaultValue=<rgba 0 0 5 1>}.RGB
   B=Power(Base=Add(A=@n26, B=@n27.RGB) {constB=-0.5}, Exp=@n27.A)
  )
  n29 = Constant()
  n30 = NamedRerouteUsage()
  n31 = VectorParameter() {group=03.WPO, parameterName=FromNoiseNormalA_OffsetAndPower, defaultValue=<rgba 0 0 0 5>}
  n32 = Multiply(
   A=VectorParameter() {group=03.WPO, parameterName=FromNoiseNormalA_Multiply, defaultValue=<rgba 2 5 2 1>}.RGB
   B=Power(Base=Add(A=@n30, B=@n31.RGB) {constB=-0.5}, Exp=@n31.A)
  )
  n33 = Multiply(A=@n32, B=ScalarParameter() {defaultValue=1.7, parameterName=ES3_WPOStrengthScale}) {constB=1.7}
  n34 = Transform(None=Add(
    A=ShadingPathSwitch(Default=@n28, Deferred=@n28, Forward=@n28, Mobile=@n29)
    B=ShadingPathSwitch(Default=@n32, Deferred=@n32, Forward=@n32, Mobile=@n33)
   ))
  n35 = VectorParameter() {parameterName=ES3_WPOOffset, defaultValue=<rgba 0 0 1 1>}
  n36 = ShadingPathSwitch(Default=@n34, Deferred=@n34, Forward=@n34, Mobile=Add(A=@n34, B=Transform(None=@n35.RGB) {transformSourceType=TRANSFORMSOURCE_Local}))
  n37 = ScalarParameter() {defaultValue=150, parameterName=WPO_Strength}
  NamedRerouteDeclaration(None=Multiply(A=NamedRerouteUsage(), B=Multiply(A=@n36, B=@n37))) {name=Final_WPO, nodeColor=<rgba 0 0.3906 1 1>}
  n38 = AppendVector(
   A=MaterialFunctionCall(Float2=NamedRerouteUsage()) {materialFunction=&BreakOutFloat2Components}.R
   B=Multiply(A=NamedRerouteUsage()) {constB=0.5}
  )
  n39 = VectorParameter() {group=00.UV, parameterName=ScaleUVByCenter, defaultValue=<rgba 1 1 0 0>}
  n40 = AppendVector(A=@n39.R, B=@n39.R)
  NamedRerouteDeclaration(None=StaticSwitchParameter(
    True=MaterialFunctionCall() {materialFunction=&VectorToRadialValue}.Radial Coordinates
    False=MaterialFunctionCall(UVs=@n38, Texture Scale=@n40) {materialFunction=&ScaleUVsByCenter}.UVs
   ) {parameterName=PortalUV}) {name=AfterProcessUV, nodeColor=<rgba 1 0.966 0 1>}
  n41 = NamedRerouteUsage()
  n42 = VectorParameter() {group=20.NoiseNormal, parameterName=NoiseNormal_A_ScaleAndSpeed, defaultValue=<rgba 0.5 0.5 0 -0.526>}
  n43 = AppendVector(A=@n42.R, B=@n42.R)
  n44 = AppendVector(A=@n42.B, B=@n42.B)
  n45 = NamedRerouteUsage()
  n46 = TextureSample(UVs=Add(A=Multiply(A=@n41, B=@n43), B=Multiply(A=@n44, B=@n45))) {
   samplerType=SAMPLERTYPE_Normal
   texture=/Game/Athena/Artemis/Environments/BuildingActors/WaterFall/Materials/Texture/T_Lowres_Simple_Perlin_N
  }
  NamedRerouteDeclaration(None=@n46.RGB) {name=NoiseNormal_A_Value, nodeColor=<rgba 1 0.8438 0 1>}
  n47 = NamedRerouteUsage()
  n48 = VectorParameter() {group=20.NoiseNormal, parameterName=NoiseNormal_B_ScaleAndSpeed, defaultValue=<rgba 2 0.5 0 -0.35>}
  n49 = AppendVector(A=@n48.R, B=@n48.R)
  n50 = AppendVector(A=@n48.B, B=@n48.B)
  n51 = NamedRerouteUsage()
  n52 = TextureSample(UVs=Add(A=Multiply(A=@n47, B=@n49), B=Multiply(A=@n50, B=@n51))) {
   samplerType=SAMPLERTYPE_Normal
   texture=&T_Lowres_Simple_Perlin_N
  }
  NamedRerouteDeclaration(None=@n52.RGB) {name=NoiseNormal_B_Value, nodeColor=<rgba 1 0.8438 0 1>}
  n53 = VectorParameter() {group=10.Mask, parameterName=RadialGradientParams, channelNames {a="", b="", g=Density, r=Radius}}
  n54 = OneMinus(None=MaterialFunctionCall(Radius=@n53.Radius, Density=@n53.Radius) {materialFunction=&RadialGradientExponential}.RadialGradientExponential)
  n55 = VectorParameter() {group=10.Mask, parameterName=RadialGradientParams_2, channelNames {a="", b="", g=Density, r=Radius}}
  n56 = MaterialFunctionCall(Radius=@n55.Radius, Density=@n55.Radius) {materialFunction=&RadialGradientExponential}
  n57 = CurveAtlasRowParameter(CurveTime=NamedRerouteUsage()) {
   atlas=&CLA_WaterFall
   curve=/Game/Athena/Artemis/Environments/BuildingActors/WaterFall/Materials/Texture/CurveAtlas/CL_WPOCurve_SpalshBottom3
   defaultValue=11
   parameterName=MaskCurve
   sortPriority=0
  }
  n58 = NamedRerouteUsage()
  n59 = CurveAtlasRowParameter(CurveTime=MaterialFunctionCall(Float2=NamedRerouteUsage()) {materialFunction=&BreakOutFloat2Components}.R) {
   atlas=&CLA_WaterFall
   curve=&CL_WPOCurve_SpalshBottom3
   defaultValue=11
   parameterName=MaskCurve
   sortPriority=0
  }
  n60 = Multiply(A=Power(Base=OneMinus(None=@n58)) {constExponent=3}, B=@n59.G)
  n61 = LinearInterpolate(A=@n57.R, B=Saturate(None=Add(A=@n60, B=@n57.R)), Alpha=ScalarParameter() {group=10.Mask, parameterName=ThicknessTweak})
  NamedRerouteDeclaration(None=StaticSwitchParameter(True=Multiply(A=@n54, B=@n56.RadialGradientExponential), False=Multiply(A=@n61, B=@n59.G)) {parameterName=PortalUV}) {name=WPOMask, nodeColor=<rgba 0.125 1 0 1>}
  n62 = NamedRerouteUsage()
  n63 = Sine(None=Multiply(A=MaterialFunctionCall(Float2=@n62) {materialFunction=&BreakOutFloat2Components}.R) {constB=0.5})
  n64 = OneMinus(None=Power(Base=OneMinus(None=@n63)) {constExponent=6})
  n65 = OneMinus(None=Power(Base=NamedRerouteUsage()) {constExponent=2.5})
  NamedRerouteDeclaration(None=StaticSwitchParameter(
    True=MaterialFunctionCall(Radius=@n53.B, Density=@n53.B) {materialFunction=&RadialGradientExponential}.RadialGradientExponential
    False=Multiply(A=@n64, B=@n65)
   ) {parameterName=PortalUV}) {name=EdgeMask, nodeColor=<rgba 1 0.5391 0 1>}
  n66 = Multiply(A=MaterialFunctionCall(Float2=NamedRerouteUsage()) {materialFunction=&BreakOutFloat2Components}.G) {constB=2}
  n67 = Subtract(A=@n66)
  NamedRerouteDeclaration(None=If(A=@n66, B=Constant() {r=1}, "A > B"=Saturate(None=@n67), "A == B"=Constant(), "A < B"=OneMinus(None=@n66))) {name=TwoDirV, nodeColor=<rgba 0.0781 1 0 1>}
  NamedRerouteDeclaration(None=TextureCoordinate()) {name=OriginUV, nodeColor=<rgba 0 1 0.0859 1>}
  n68 = NamedRerouteUsage()
  n69 = VectorParameter() {group=21.Noise, parameterName=Noise_A_ScaleAndSpeed, defaultValue=<rgba 2 0.2 0 0.2>}
  n70 = AppendVector(A=@n69.R, B=@n69.R)
  n71 = AppendVector(A=@n69.B, B=@n69.B)
  n72 = NamedRerouteUsage()
  n73 = TextureSample(UVs=Add(A=Multiply(A=@n68, B=@n70), B=Multiply(A=@n71, B=@n72))) {
   samplerType=SAMPLERTYPE_LinearColor
   texture=/Game/Athena/Artemis/Environments/BuildingActors/WaterFall/Materials/Texture/T_MergedNoiseTexture
  }
  n74 = AppendVector(A=@n73.B, B=@n73.B)
  n75 = VectorParameter() {group=21.Noise, parameterName=Noise_A_Strength, defaultValue=<rgba 0 0.15 0.15 1>}
  NamedRerouteDeclaration(None=Multiply(A=Add(A=@n74) {constB=-0.5}, B=AppendVector(A=@n75.R, B=@n75.R))) {name=Noise_A_Value, nodeColor=<rgba 0 0.6953 1 1>}
  NamedRerouteUsage()
  NamedRerouteUsage()
  Reroute(None=@n17)
  n76 = DynamicParameter() {paramNames [FadeDistanceADd, FadeDistancePowerAd, Param3, Param4]}
  n77 = ScalarParameter() {defaultValue=200, parameterName=CameraDistance}
  n78 = Power(
   Base=MaterialFunctionCall(Distance=Add(A=@n76.FadeDistanceADd, B=@n77)) {materialFunction=&Near_Camera_Fade}.Result
   Exp=Add(B=@n76.FadeDistancePowerAd) {constA=4}
  ) {constExponent=5}
  n79 = Constant() {r=0.5}
  n80 = MaterialFunctionCall(Alpha Threshold=@n78, Random=@n79) {materialFunction=&DitherTemporalAA_Fortnite}
  n81 = MaterialFunctionCall(Alpha Threshold=NamedRerouteUsage(), Random=@n79) {materialFunction=&DitherTemporalAA_Fortnite}
  ShadowReplace(Default=Multiply(A=LinearInterpolate(A=@n80.Result, Alpha=@n78), B=@n81.Result), Shadow=@n81.Result)
 ]
}
