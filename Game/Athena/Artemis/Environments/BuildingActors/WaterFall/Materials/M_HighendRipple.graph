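# Translucent lit, refractive water ripple material for the waterfall environment, enabled for Niagara mesh particles.
# A curve atlas remaps radial distance to drive two summed, time-scrolled sine waves, which supply height-derived world-space normals and vertex-normal displacement.
# Radial masks fade displacement and opacity toward the rim, with opacity also softened by depth fade, view angle and camera proximity.
# A second depth fade hue-shifts the base color near intersecting geometry.
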Material /Game/Athena/Artemis/Environments/BuildingActors/WaterFall/Materials/M_HighendRipple
graph {
 properties {
  bUsedWithNiagaraMeshParticles=true
  blendMode=BLEND_Translucent
  pixelDepthOffsetMode=PDOM_Legacy
  refractionMethod=RM_IndexOfRefraction
  translucencyLightingMode=TLM_Surface
  usedShadingModels=Default Lit
 }
 nodes [
  n1 = ScalarParameter() {defaultValue=1, parameterName=TotalRippleUVScale}
  n2 = MaterialFunctionCall(Float2=MaterialFunctionCall() {materialFunction=&VectorToRadialValue}.Radial Coordinates) {materialFunction=&BreakOutFloat2Components}
  n3 = CurveAtlasRowParameter(CurveTime=@n2.G) {
   atlas=/Game/Athena/Artemis/Environments/BuildingActors/WaterFall/Materials/Texture/CurveAtlas/CLA_WaterFall
   curve=/Game/Athena/Artemis/Environments/BuildingActors/WaterFall/Materials/Texture/CurveAtlas/CL_Ripple2
   defaultValue=14
   parameterName=RippleCurve
  }
  n4 = VectorParameter() {parameterName=WaveParam, channelNames {a=WaveBSpeed, b=WaveASpeed, g=WaveBFrequency, r=WaveAFrequency}, defaultValue=<rgba 7 2 -1.5 -1>}
  n5 = Multiply(A=Time() {bOverride_Period=true, period=333}, B=ScalarParameter() {defaultValue=0.7, parameterName=TimeScale})
  n6 = Add(A=Multiply(A=@n3.R, B=@n4.WaveBFrequency) {constB=7}, B=Multiply(A=@n5, B=@n4.WaveBSpeed) {constB=-1})
  n7 = Add(A=Multiply(A=@n3.R, B=@n4.WaveAFrequency) {constB=7}, B=Multiply(A=@n5, B=@n4.WaveASpeed) {constB=-1})
  n8 = Add(A=Sine(None=Multiply(A=@n1, B=@n6)), B=Sine(None=Multiply(A=@n7, B=@n1)))
  Add(A=MaterialFunctionCall!(Height=Multiply(A=@n8) {constA=20, constB=20}, Absolute World Position, World Space Vertex Normals) {materialFunction=/Engine/Functions/Engine_MaterialFunctions03/Procedurals/HeightToNormalSmooth}.WorldSpaceNormal) {constB=0}
  n9 = VectorParameter() {parameterName=BaseParam2, channelNames {a="", b=HueShiftDF, g=WPOScale, r=OpacityDF}, defaultValue=<rgba 100 20 300 0>}
  n10 = DepthFade(FadeDistance=@n9.HueShiftDF)
  MaterialFunctionCall(
   Hue Shift Percentage=Multiply(A=OneMinus(None=@n10), B=ScalarParameter() {defaultValue=-0.1, parameterName=HueShiftStrength})
   Texture=VectorParameter() {parameterName=Color, defaultValue=<rgba 0 0.1333 0.2 1>}.RGB
  ) {materialFunction=&HueShift}
  n11 = Subtract(A=DotProduct(A=VertexNormalWS(), B=CameraVectorWS())) {constB=0.1}
  n12 = OneMinus(None=Saturate(None=@n2.G))
  n13 = VectorParameter() {parameterName=MaskParam, channelNames {a="", b="", g=WPOMaskSharpness, r=OpacityMaskSharpness}, defaultValue=<rgba 2 5 0 0>}
  n14 = Saturate(None=Multiply(A=@n12, B=@n13.WPOMaskSharpness))
  n15 = VectorParameter() {parameterName=BaseParam, channelNames {a=Opacity, b=Roughness, g=Specular, r=Metal}, defaultValue=<rgba 0 0.1 0 0.5>}
  n16 = DepthFade(Opacity=@n15.Opacity, FadeDistance=@n9.OpacityDF)
  n17 = ScalarParameter() {defaultValue=200, parameterName=NearCameraFade}
  Multiply(
   A=Saturate(None=Multiply(A=@n11) {constB=5})
   B=Multiply(
    A=Multiply(A=@n14, B=@n16)
    B=MaterialFunctionCall(Distance=@n17) {materialFunction=&Near_Camera_Fade}.Result
   )
  )
  n18 = VertexNormalWS()
  Multiply(A=Saturate(None=Multiply(A=@n12, B=@n13.OpacityMaskSharpness)), B=Multiply(A=@n9.WPOScale, B=Multiply(A=@n8, B=@n18)) {constA=20})
  ScalarParameter() {defaultValue=1, parameterName=Ref}
 ]
}
