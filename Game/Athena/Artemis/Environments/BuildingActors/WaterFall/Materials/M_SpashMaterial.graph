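# M_SpashMaterial is a translucent lit Niagara sprite material for waterfall splash ripples and foam.
# Layered, time-scrolled noise distorts radial UVs sampling a wave-foam texture for normals and opacity, with curve-atlas rows controlling dissolve, contrast, normal flattening and color.
# A bubble texture adds foam color, while particle color and dynamic parameters feed color and fake emissive lighting.
# Depth and near-camera fades soften opacity, and attenuated Gerstner water waves supply world position offset.
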Material /Game/Athena/Artemis/Environments/BuildingActors/WaterFall/Materials/M_SpashMaterial
graph {
 properties {
  bUsedWithNiagaraSprites=true
  blendMode=BLEND_Translucent
  pixelDepthOffsetMode=PDOM_Legacy
  translucencyLightingMode=TLM_SurfacePerPixelLighting
  usedShadingModels=Default Lit
 }
 nodes [
  Constant() {r=1}
  n1 = NamedRerouteUsage()
  n2 = VectorParameter() {group=20.NoiseA, parameterName=NoiseA_ScaleAndOffset, defaultValue=<rgba 6 0.7 0 -0.6>}
  n3 = AppendVector(A=@n2.R, B=@n2.R)
  n4 = AppendVector(A=@n2.B, B=@n2.B)
  n5 = NamedRerouteUsage()
  n6 = TextureCoordinate()
  n7 = TextureSample!(UVs=Add(A=Multiply(A=@n1, B=@n3), B=Multiply(A=@n4, B=@n5)), Tex, "DDX(UVs)"=DDX(Value=@n6), "DDY(UVs)"=DDY(Value=@n6), Apply View MipBias) {
   mipValueMode=TMVM_Derivative
   samplerType=SAMPLERTYPE_LinearColor
   texture=/Game/Athena/Artemis/Environments/BuildingActors/WaterFall/Materials/Texture/T_MergedFoamTexture
  }
  n8 = Add(A=AppendVector(A=@n7.R, B=@n7.R)) {constB=-0.5}
  n9 = Multiply(A=@n8) {constB=30}
  n10 = NamedRerouteUsage()
  n11 = VectorParameter() {group=22.NoiseC, parameterName=NoiseC_ScaleAndOffset, defaultValue=<rgba 1 2 0 -0.25>}
  n12 = AppendVector(A=@n11.R, B=@n11.R)
  n13 = AppendVector(A=@n11.B, B=@n11.B)
  n14 = NamedRerouteUsage()
  n15 = TextureCoordinate()
  n16 = TextureSample(UVs=Add(A=Multiply(A=@n10, B=@n12), B=Multiply(A=@n13, B=@n14)), "DDX(UVs)"=DDX(Value=@n15), "DDY(UVs)"=DDY(Value=@n15)) {
   mipValueMode=TMVM_Derivative
   samplerType=SAMPLERTYPE_LinearColor
   texture=&T_MergedFoamTexture
  }
  n17 = AppendVector(A=@n16.B, B=@n16.B)
  n18 = Multiply(A=Add(A=@n17) {constB=-0.5}) {constB=70}
  n19 = WorldPosition()
  n20 = Constant()
  n21 = AppendVector(A=Add(A=Add(A=@n9, B=@n18), B=ComponentMask(None=@n19.XYZ) {g=true, r=true}), B=@n20)
  n22 = CollectionParameter() {collection=&FortniteMaterialParameters, parameterName=PlayerPosition}
  n23 = AppendVector(A=ComponentMask(None=@n22) {g=true, r=true}, B=@n20)
  n24 = VectorParameter() {parameterName=PlayerPosMask, defaultValue=<rgba 60 0.15 1 1>}
  n25 = OneMinus(None=Saturate(None=SphereMask(A=@n21, B=@n23, Radius=@n24.R, Hardness=@n24.R)))
  Multiply(A=SmoothStep(Value=@n25) {constValue=1}, B=NamedRerouteUsage())
  n26 = WorldPosition()
  n27 = MaterialFunctionCall!(Texture, Position=@n26.XYZ) {materialFunction=/FortWater/Materials/Functions/WaterBodyData}
  n28 = Divide(
   A=Subtract(A=@n27.VelocityAndLength, B=ScalarParameter() {defaultValue=8, parameterName=Velocity for Full Waves})
   B=ScalarParameter() {defaultValue=64, parameterName=Velocity for No Waves}
  )
  n29 = Multiply(
   A=OneMinus(None=Saturate(None=@n28))
   B=MaterialFunctionCall!(WaterDepth=@n27.Water Depth, Water Body Index=ScalarParameter() {defaultValue=48, parameterName=WaterBodyIndex}) {materialFunction=/Water/Materials/Functions/ComputeWavesAttenuationFactor}.WaveAttenuationFactor
  )
  n30 = MaterialFunctionCall!(
   Z Offset Only
   Position=@n26.XYZ
   WaterBodyIndex=ScalarParameter() {defaultValue=48, parameterName=WaterBodyIndex}
   Normals use VS Interpolation
   All Normals VS
  ) {materialFunction=/FortWater/Materials/Functions/Fort_ComputeGerstnerWaves}
  NamedRerouteDeclaration(None=Add(B=Multiply(A=@n29, B=@n30.WPO) {constA=1})) {name=Final_WPO, nodeColor=<rgba 1 0 0.1875 1>}
  NamedRerouteDeclaration(None=MaterialFunctionCall() {materialFunction=&VectorToRadialValue}.Radial Coordinates) {name=UV, nodeColor=<rgba 1 0.4922 0 1>}
  n31 = CurveAtlasRowParameter(CurveTime=MaterialFunctionCall(Float2=NamedRerouteUsage()) {materialFunction=&BreakOutFloat2Components}.G) {
   atlas=/Game/Athena/Artemis/Environments/BuildingActors/WaterFall/Materials/Texture/CurveAtlas/CLA_WaterFall
   curve=/Game/Athena/Artemis/Environments/BuildingActors/WaterFall/Materials/Texture/CurveAtlas/CL_RippleCurve
   defaultValue=12
   parameterName=Dissolve/Contrast//FlatNormal
   sortPriority=0
  }
  NamedRerouteDeclaration(None=@n31.G) {name=Curve_Contrast, nodeColor=<rgba 0 1 0.5078 1>}
  NamedRerouteDeclaration(None=@n31.R) {name=Curve_Dissolve, nodeColor=<rgba 1 0 0.9609 1>}
  n32 = TextureCoordinate()
  n33 = Subtract(A=OneMinus(None=MaterialFunctionCall(Float2=@n32) {materialFunction=&BreakOutFloat2Components}.G)) {constB=0.5}
  NamedRerouteDeclaration(None=Saturate(None=Multiply(A=@n33) {constB=100})) {name=AreaMask, nodeColor=<rgba 1 0.2578 0 1>}
  n34 = Subtract(A=ObjectPositionWS(), B=CameraPositionWS())
  n35 = Transform(None=Constant3Vector() {constant=<rgba 0 0 -1 0>}) {transformSourceType=TRANSFORMSOURCE_Local}
  n36 = Subtract(A=DotProduct(A=Normalize(VectorInput=@n34), B=Normalize(VectorInput=@n35))) {constB=0}
  NamedRerouteDeclaration(None=Multiply(A=Saturate(None=@n36), B=VectorParameter() {group=10.Color, parameterName=Color_FakeSHLighting, defaultValue=<rgba 0 0.0303 0.035 0>}.RGB)) {name=FakeSHLighting, nodeColor=<rgba 0 1 0.4609 1>}
  n37 = ParticleColor()
  n38 = VectorParameter() {group=10.Color, parameterName=Color_FakeCaveLighting, defaultValue=<rgba 0 0 0 1>}
  n39 = DynamicParameter() {parameterIndex=3, defaultValue=<rgba 1 1 1 0>, paramNames [Param1, Param2, FakeSHLightingScale, FakeSHLightingDesat]}
  n40 = Multiply(
   A=LinearInterpolate(Alpha=NamedRerouteUsage())
   B=Multiply(A=Desaturation(None=@n38.RGB, Fraction=@n39.FakeSHLightingDesat), B=@n39.FakeSHLightingScale)
  )
  NamedRerouteDeclaration(None=Add(A=NamedRerouteUsage(), B=Add(A=@n37.RGB, B=@n40))) {name=Final_Emissive, nodeColor=<rgba 0 0.6016 1 1>}
  n41 = DynamicParameter()
  n42 = VectorParameter() {group=10.Color, parameterName=Color_EdgeColor, defaultValue=<rgba 0 0 0 1>}
  n43 = CurveAtlasRowParameter(CurveTime=MaterialFunctionCall(Float2=NamedRerouteUsage()) {materialFunction=&BreakOutFloat2Components}.G) {
   atlas=&CLA_WaterFall
   curve=/Game/Athena/Artemis/Environments/BuildingActors/WaterFall/Prefab/S21/CL_RippleCurve_Color
   defaultValue=13
   parameterName=ColorCurve
   sortPriority=0
  }
  n44 = Desaturation(None=LinearInterpolate(A=@n43, B=Multiply(A=@n43, B=@n42.RGB), Alpha=NamedRerouteUsage()), Fraction=ParticleColor().A)
  n45 = VectorParameter() {group=03.Foam, parameterName=FoamGradient, defaultValue=<rgba 0.12 5 0 0>}
  n46 = Max(
   A=MaterialFunctionCall(Radius=@n45.R, Density=@n45.R) {materialFunction=&RadialGradientExponential}.RadialGradientExponential
   B=MaterialFunctionCall(Distance=ScalarParameter() {defaultValue=3000, group=03.Foam, parameterName=FoamDistance}) {materialFunction=&Near_Camera_Fade}.Result
  )
  n47 = VectorParameter() {group=03.Foam, parameterName=FoamTextureScale, defaultValue=<rgba 10 3.5 0 -1.3>}
  n48 = Multiply(
   A=NamedRerouteUsage()
   B=MaterialFunctionCall(R=@n47.R, G=@n47.R) {materialFunction=&AppendMany}.RG
  )
  n49 = NamedRerouteUsage()
  n50 = Multiply(A=@n49, B=AppendVector(A=@n47.B, B=@n47.B))
  n51 = NamedRerouteUsage()
  n52 = VectorParameter() {group=03.Foam, parameterName=NoiseD_ScaleAndOffset, defaultValue=<rgba 8 1 0 -0.5>}
  n53 = AppendVector(A=@n52.R, B=@n52.R)
  n54 = AppendVector(A=@n52.B, B=@n52.B)
  n55 = NamedRerouteUsage()
  n56 = TextureSample!(UVs=Add(A=Multiply(A=@n51, B=@n53), B=Multiply(A=@n54, B=@n55)), Tex, Apply View MipBias) {
   samplerType=SAMPLERTYPE_LinearColor
   texture=/Game/Athena/Artemis/Environments/BuildingActors/WaterFall/Materials/Texture/T_MergedNoiseTexture
  }
  n57 = VectorParameter() {group=03.Foam, parameterName=FoamTextureNoiseStrength, defaultValue=<rgba 0.2 0.2 0.05 0.07>}
  n58 = Multiply(A=AppendVector(A=@n56.A, B=@n56.A), B=AppendVector(A=@n57.R, B=@n57.R))
  n59 = Multiply(A=@n17, B=AppendVector(A=@n57.B, B=@n57.B))
  n60 = TextureCoordinate()
  n61 = DDX(Value=@n60)
  n62 = DDY(Value=@n60)
  n63 = TextureSample!(UVs=Add(A=Add(A=@n48, B=@n50), B=Add(A=@n58, B=@n59)), Tex, "DDX(UVs)"=@n61, "DDY(UVs)"=@n62, Apply View MipBias) {
   mipValueMode=TMVM_Derivative
   samplerType=SAMPLERTYPE_LinearGrayscale
   texture=/Game/Gadgets/Assets/VinderTech_GliderChute/Glider_Thrive_Female/FX/Texture/T_CloudBubble
  }
  NamedRerouteDeclaration(None=LinearInterpolate(
    A=VectorParameter() {group=03.Foam, parameterName=FoamTextureColor, defaultValue=<rgba 0.075 0.1917 0.25 1>}.RGB
    B=Multiply(A=@n41.Param1, B=@n44)
    Alpha=LinearInterpolate(A=@n63.R, Alpha=@n46)
   )) {name=Final_BaseColor, nodeColor=<rgba 1 0 0.9844 1>}
  NamedRerouteDeclaration(None=Multiply(A=Time() {bOverride_Period=true, period=333}, B=ScalarParameter() {defaultValue=0.73, parameterName=TimeScale})) {name=Time, nodeColor=<rgba 1 0.5156 0 1>}
  n64 = MaterialFunctionCall(Float2=NamedRerouteUsage()) {materialFunction=&BreakOutFloat2Components}
  n65 = VectorParameter() {group=20.NoiseA, parameterName=NoiseA_StrengthAndMaskProperty, defaultValue=<rgba 0 0.2 4.5 2>}
  n66 = Saturate(None=Power(Base=OneMinus(None=@n64.G), Exp=@n65.B) {constExponent=3})
  n67 = Saturate(None=Multiply(A=Subtract(A=@n66) {constB=0.05}, B=@n65.A))
  n68 = Multiply(A=Multiply(A=@n8) {constB=1.7}, B=AppendVector(A=@n65.R, B=@n65.R))
  n69 = VectorParameter() {group=20.NoiseA, parameterName=NoiseA_StrengthOffset, defaultValue=<rgba 0 0.25 0 0>}
  n70 = AppendVector(A=@n69.R, B=@n69.R)
  n71 = TextureCoordinate()
  n72 = VectorParameter() {group=21.NoiseB, parameterName=NoiseB_ScaleAndOffset, defaultValue=<rgba 0.8 0.8 -0.02 0.07>}
  n73 = AppendVector(A=@n72.R, B=@n72.R)
  n74 = AppendVector(A=@n72.B, B=@n72.B)
  n75 = NamedRerouteUsage()
  n76 = TextureSample!(UVs=Add(A=Multiply(A=@n71, B=@n73), B=Multiply(A=@n74, B=@n75)), Tex, Apply View MipBias) {samplerType=SAMPLERTYPE_Normal, texture=/Game/Characters/Player/Male/Large/Bodies/M_LRG_Clash_V/Textures/Simple_Perlin_N}
  n77 = VectorParameter() {group=21.NoiseB, parameterName=NoiseB_Strength, defaultValue=<rgba 0.01 0.01 0 0>}
  n78 = Multiply(A=AppendVector(A=@n76.R, B=@n76.R), B=AppendVector(A=@n77.R, B=@n77.R))
  n79 = Add(A=Multiply(A=@n67, B=Add(A=@n68, B=@n70)), B=ShadingPathSwitch(Default=@n78, Deferred=@n78, Forward=@n78, Mobile=Constant()))
  n80 = VectorParameter() {group=22.NoiseC, parameterName=NoiseC_StrengthAndOffset, defaultValue=<rgba 0.01 0.01 0 0>}
  n81 = AppendVector(A=@n80.R, B=@n80.R)
  n82 = Multiply(A=OneMinus(None=@n67), B=Add(A=Multiply(A=@n17, B=@n81), B=AppendVector(A=@n80.B, B=@n80.B)))
  n83 = VectorParameter() {group=01.TextureSample, parameterName=UVScaleAndOffset, defaultValue=<rgba 0.35 1 0 -0.2>}
  n84 = Power(
   Base=MaterialFunctionCall(Float2=NamedRerouteUsage()) {materialFunction=&BreakOutFloat2Components}.G
   Exp=ScalarParameter() {group=01.TextureSample, parameterName=UVScaleGradientPower}
  )
  n85 = Multiply(A=NamedRerouteUsage(), B=AppendVector(A=@n83.R, B=Multiply(A=@n83.G, B=@n84)))
  n86 = Multiply(A=AppendVector(A=@n83.B, B=@n83.B), B=@n49)
  n87 = TextureSample!(UVs=Add(A=Add(A=@n79, B=@n82), B=Add(A=@n85, B=@n86)), Tex, "DDX(UVs)"=@n61, "DDY(UVs)"=@n62, Apply View MipBias) {
   mipValueMode=TMVM_Derivative
   samplerType=SAMPLERTYPE_LinearColor
   texture=/Game/Athena/Artemis/Environments/BuildingActors/WaterFall/Materials/Texture/T_MergedWaveFoamTexture
  }
  n88 = Multiply(A=Add(A=AppendVector(A=@n87.R, B=@n87.R)) {constB=-0.5}) {constB=2}
  NamedRerouteDeclaration(None=MaterialFunctionCall(Normal=DeriveNormalZ(InXY=@n88), Flatness=NamedRerouteUsage()) {materialFunction=&FlattenNormal}.Result) {name=Final_Normal, nodeColor=<rgba 1 0.9141 0 1>}
  n89 = Add(A=@n87.A, B=Constant() {r=1})
  n90 = NamedRerouteUsage()
  n91 = NamedRerouteUsage()
  n92 = ScalarParameter() {defaultValue=1, group=02.Opacity, parameterName=ContarstScale}
  n93 = Saturate(None=MaterialFunctionCall(In=Subtract(A=@n89, B=@n90), Contrast=Multiply(A=@n91, B=@n92)) {materialFunction=&CheapContrast}.Result)
  n94 = ScalarParameter() {group=02.Opacity, parameterName=DistanceFade}
  n95 = ScalarParameter() {defaultValue=200, group=02.Opacity, parameterName=NearCameraFade}
  NamedRerouteDeclaration(None=Multiply(
    A=DepthFade(Opacity=@n93, FadeDistance=@n94)
    B=MaterialFunctionCall(Distance=@n95) {materialFunction=&Near_Camera_Fade}.Result
   )) {name=Final_Opacity, nodeColor=<rgba 0 1 0.5547 1>}
  NamedRerouteDeclaration(None=@n31.A) {name=Curve_NormalFlat, nodeColor=<rgba 0 1 0.5078 1>}
  NamedRerouteUsage()
  NamedRerouteUsage()
  NamedRerouteUsage()
  NamedRerouteUsage()
  Power(Base=@n25)
 ]
}
