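# Translucent, two-sided lit material for Niagara sprites that renders flowing water for the waterfall.
# Panned noise samples, some with noise-distorted UVs, combine with a vertical gradient to drive opacity and a blend between highlight and dark colors.
# UV-based masks fade the sides and the low-V end, and DepthFade softens intersections.
# A ScaleU dynamic parameter lets particles scale the flow UVs horizontally around their center.
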Material /Game/Athena/Artemis/Environments/BuildingActors/WaterFall/Materials/M_WaterFlow
graph {
 properties {
  bUsedWithNiagaraSprites=true
  blendMode=BLEND_Translucent
  pixelDepthOffsetMode=PDOM_Legacy
  translucencyLightingMode=TLM_Surface
  twoSided=true
  usedShadingModels=Default Lit
 }
 nodes [
  n1 = TextureCoordinate()
  n2 = MaterialFunctionCall(UVs=@n1, Texture Scale=AppendVector(A=DynamicParameter() {paramNames [ScaleU,Param2,Param3,Param4]}.ScaleU, B=Constant() {r=1})) {materialFunction=&ScaleUVsByCenter}
  n3 = Panner(Coordinate=@n2.UVs) {speedY=0.15}
  Add(A=@n3, B=Constant2Vector() {g=0.5, r=0.75})
  n4 = MaterialFunctionCall(Float2=OneMinus(None=@n1)) {materialFunction=&BreakOutFloat2Components}
  n5 = Saturate(None=Multiply(A=Add(A=@n4.G) {constB=-0.8}) {constB=10})
  n6 = Max(A=@n4.R, B=@n4.R)
  n7 = MaterialFunctionCall(Float2=@n1) {materialFunction=&BreakOutFloat2Components}
  n8 = Max(A=@n7.R, B=@n7.R)
  n9 = OneMinus(None=Max(A=Power(Base=@n6) {constExponent=3}, B=Power(Base=@n8) {constExponent=3}))
  n10 = TextureSample(UVs=@n3) {samplerType=SAMPLERTYPE_Grayscale, texture=/FortWater/Textures/LowResBlurredNoise}
  n11 = MaterialFunctionCall(Float4=Constant4Vector() {constant=<rgba 2 1 0 0.5>}.RGBA) {materialFunction=&BreakOutFloat4Components}
  n12 = AppendVector(A=@n11.R, B=@n11.R)
  n13 = Time() {bOverride_Period=true, period=300}
  n14 = TextureSample(UVs=Panner(Coordinate=Multiply(A=@n2.UVs, B=@n12), Time=@n13, Speed=AppendVector(A=@n11.B, B=@n11.B)) {speedY=0.5}) {samplerType=SAMPLERTYPE_Grayscale, texture=&LowResBlurredNoise}
  n15 = AppendVector(A=@n14.R, B=@n14.R)
  n16 = Add(A=Multiply(A=Saturate(None=@n15), B=Constant2Vector() {g=0.3}) {constB=0.05}, B=Panner(Coordinate=@n2.UVs, Time=@n13) {speedY=0.5})
  n17 = Constant2Vector() {g=0.5, r=0.25}
  n18 = MaterialFunctionCall(Float4=Constant4Vector() {constant=<rgba 5 1 0 0.5>}.RGBA) {materialFunction=&BreakOutFloat4Components}
  n19 = AppendVector(A=@n18.R, B=@n18.R)
  n20 = TextureSample(UVs=Panner(Coordinate=Multiply(A=@n2.UVs, B=@n19), Time=@n13, Speed=AppendVector(A=@n18.B, B=@n18.B)) {speedY=0.3}) {
   samplerType=SAMPLERTYPE_LinearColor
   texture=/Game/Athena/Artemis/Environments/BuildingActors/WaterFall/Materials/Texture/T_MergedNoiseTexture
  }
  n21 = MaterialFunctionCall(Float4=Constant4Vector() {constant=<rgba 0.1 0.1 0 0>}.RGBA) {materialFunction=&BreakOutFloat4Components}
  n22 = Multiply(A=Saturate(None=AppendVector(A=@n20.A, B=@n20.A)), B=AppendVector(A=@n21.R, B=@n21.R)) {constB=0.05}
  n23 = TextureSample(UVs=Add(A=Add(A=@n16, B=@n17), B=Multiply(A=@n22, B=@n21.B))) {samplerType=SAMPLERTYPE_Grayscale, texture=&LowResBlurredNoise}
  n24 = OneMinus(None=MaterialFunctionCall(Float2=@n2.UVs) {materialFunction=&BreakOutFloat2Components}.G)
  n25 = Multiply(A=Add(A=Multiply(A=@n10.B, B=@n23.B), B=Power(Base=@n24) {constExponent=3.5})) {constB=2}
  n26 = MaterialFunctionCall(In=@n25, Contrast=Constant() {r=0.1}) {materialFunction=&CheapContrast}
  n27 = TextureSample(UVs=Add(A=@n22, B=Panner(Time=@n13) {speedY=0.5})) {samplerType=SAMPLERTYPE_Grayscale, texture=&LowResBlurredNoise}
  n28 = Multiply(A=Power(Base=Saturate(None=@n26.Result)), B=Add(A=@n27.R, B=@n27.R))
  DepthFade(Opacity=Multiply(A=OneMinus(None=@n5), B=Multiply(A=@n9, B=@n28))) {fadeDistanceDefault=20}
  LinearInterpolate(
   A=VectorParameter() {parameterName=HighLightColor, defaultValue=<rgba 0.16 0.4 0.24 1>}.RGB
   B=VectorParameter() {parameterName=DarkColor, defaultValue=<rgba 0.2 0.4 0.4 1>}.RGB
   Alpha=Saturate(None=@n25)
  )
 ]
}
