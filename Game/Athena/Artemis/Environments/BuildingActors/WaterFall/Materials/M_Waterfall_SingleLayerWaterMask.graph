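# Masked, two-sided Single Layer Water material for waterfall building actors, named as a water mask.
# An inverted, powered Fresnel term feeds a temporal AA dither function and, separately, multiplies an inverted, offset UV-horizontal gradient; no root connection appears for either.
# Parameters drive the water output's scattering and behind-water color.
# One scalar "PG" feeds both absorption and phase G, while a same-named vector parameter sits unconnected.
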Material /Game/Athena/Artemis/Environments/BuildingActors/WaterFall/Materials/M_Waterfall_SingleLayerWaterMask
graph {
 properties {
  blendMode=BLEND_Masked
  pixelDepthOffsetMode=PDOM_Legacy
  shadingModel=MSM_SingleLayerWater
  twoSided=true
  usedShadingModels=SingleLayerWater
 }
 nodes [
  Constant()
  n1 = Fresnel(ExponentIn=Constant() {r=1.5})
  n2 = ScalarParameter() {defaultValue=6, parameterName=s}
  n3 = Multiply(A=OneMinus(None=Power(Base=@n1, Exp=@n2)), B=Constant() {r=0.65})
  MaterialFunctionCall(Alpha Threshold=@n3) {materialFunction=&DitherTemporalAA_Fortnite}
  n4 = Add(A=MaterialFunctionCall(Float2=TextureCoordinate()) {materialFunction=&BreakOutFloat2Components}.R) {constB=-0.2}
  Multiply(A=OneMinus(None=Saturate(None=@n4)), B=@n3)
  n5 = ScalarParameter() {parameterName=PG}
  SingleLayerWaterMaterialOutput!(
   ScatteringCoefficients=ScalarParameter() {parameterName=SC}
   AbsorptionCoefficients=@n5
   PhaseG=@n5
   ColorScaleBehindWater=VectorParameter() {parameterName=CSBW, defaultValue=<rgba 0.3 0.7667 1 1>}.RGB
  )
  VectorParameter() {parameterName=PG, defaultValue=<rgba 0.001 0 0 1>}
 ]
}
